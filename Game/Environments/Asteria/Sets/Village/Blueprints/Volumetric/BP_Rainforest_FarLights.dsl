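; Village environment Blueprint that fakes distant lights with one instanced static mesh.
; Set_InstancesFromTargetActors gathers actors of target classes within MaxRadius, anchoring each transform on its light, Niagara component or actor location plus rotated offset, with distance-lerped, jittered scale.
; AddInstances spawns instances from that list and a Locations array, writing index-seeded random custom data for stable per-instance material variation.
; Construction applies the material and adds instances; runtime events are empty.
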(event EventBeginPlay)

(event Collision|EventActorBeginOverlap
  (params (OtherActor :type Actor%20Object%20Reference)))

(event EventTick
  (params (DeltaSeconds :type Float%20%28single-precision%29)))

(fn AddInstances
  (bind _instancedstaticmesh (Variables|Default|GetInstancedStaticMesh))
  (bind _instancedstaticmesh_2 (Variables|Default|GetInstancedStaticMesh))
  (Utilities|FlowControl|Sequence
    (:then_0
      (for (_array_element _array_index) (Variables|Default|GetTransformList)
        (bind _returnvalue (Components|InstancedStaticMesh|AddInstance _instancedstaticmesh_2 _array_element true))
        (Components|InstancedStaticMesh|SetCustomDataValue _instancedstaticmesh_2 _returnvalue 0 (Math|Random|RandomFloatinRangefromStream (Math|Random|MakeRandomStream _array_index) 0.0 1.0) true)))
    (:then_1
      (for (_array_element_2 _array_index_2) (Variables|Default|GetLocations)
        (bind _returnvalue_2 (Components|InstancedStaticMesh|AddInstance _instancedstaticmesh (Math|Transform|MakeTransform (.location _array_element_2) (.rotation _array_element_2) "0.010000,0.010000,0.010000") false))
        (Components|InstancedStaticMesh|SetCustomDataValue _instancedstaticmesh _returnvalue_2 0 (Math|Random|RandomFloatinRangefromStream (Math|Random|MakeRandomStream _array_index_2) 0.0 1.0) true)))))

(fn Set_InstancesFromTargetActors
  (bind _returnvalue (Transformation|GetActorLocation))
  (bind _maxradius (Variables|Default|GetMaxRadius))
  (Utilities|Array|Clear (Variables|Default|GetTransformList))
  (for _array_element (Variables|Default|GetTargetActorClass)
    (bind _outactors (Actor|GetAllActorsOfClass _array_element))
    (for (_array_element_2 _array_index) _outactors
      (if (< (Math|Vector|Distance(Vector) (Transformation|GetActorLocation _array_element_2) _returnvalue) _maxradius)
        (if (Utilities|Array|IsNotEmpty (Actor|GetComponentsByClass _array_element_2 "/Script/Engine.LightComponent"))
          (LocalVariables|SetBaseLocation (Transformation|GetWorldLocation (Utilities|Array|Get(acopy) (Actor|GetComponentsByClass _array_element_2 "/Script/Engine.LightComponent"))))
          (elif (Utilities|Array|IsNotEmpty (Actor|GetComponentsByClass _array_element_2 "/Script/Niagara.NiagaraComponent"))
            (LocalVariables|SetBaseLocation (Transformation|GetWorldLocation (Utilities|Array|Get(acopy) (Actor|GetComponentsByClass _array_element_2 "/Script/Niagara.NiagaraComponent"))))
            (else
              (LocalVariables|SetBaseLocation (.location (Transformation|GetActorTransform _array_element_2))))))
        (Utilities|Array|Add (Variables|Default|GetTransformList) (Math|Transform|MakeTransform (+ (LocalVariables|GetBaseLocation) (Math|Vector|RotateVector (Variables|Default|GetOffset) (.rotation (Transformation|GetActorTransform _array_element_2)))) (.rotation (Transformation|GetActorTransform _array_element_2)) (Math|Vector|MakeVector (+ (Math|Float|Lerp (Variables|Default|GetScale) (Variables|Default|GetScale_Edge) (/ (Math|Vector|Distance(Vector) (Transformation|GetActorLocation _array_element_2) _returnvalue) _maxradius)) (* (- (Math|Random|RandomFloatinRangefromStream (Math|Random|MakeRandomStream _array_index) 0.0 1.0) 0.5) (Variables|Default|GetScaleRandom))) (+ (Math|Float|Lerp (Variables|Default|GetScale) (Variables|Default|GetScale_Edge) (/ (Math|Vector|Distance(Vector) (Transformation|GetActorLocation _array_element_2) _returnvalue) _maxradius)) (* (- (Math|Random|RandomFloatinRangefromStream (Math|Random|MakeRandomStream _array_index) 0.0 1.0) 0.5) (Variables|Default|GetScaleRandom))) (+ (Math|Float|Lerp (Variables|Default|GetScale) (Variables|Default|GetScale_Edge) (/ (Math|Vector|Distance(Vector) (Transformation|GetActorLocation _array_element_2) _returnvalue) _maxradius)) (* (- (Math|Random|RandomFloatinRangefromStream (Math|Random|MakeRandomStream _array_index) 0.0 1.0) 0.5) (Variables|Default|GetScaleRandom))))))))
    (Components|InstancedStaticMesh|ClearInstances (Variables|Default|GetInstancedStaticMesh))
    (CallFunction|AddInstances)))

(fn ConstructionScript
  (Rendering|Material|SetMaterial (Variables|Default|GetInstancedStaticMesh) 0 (Variables|Default|GetMaterial))
  (CallFunction|AddInstances))
