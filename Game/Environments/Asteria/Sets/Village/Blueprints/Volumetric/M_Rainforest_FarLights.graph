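# Additive unlit material for distant village lights in the rainforest set, drawing instanced flares and ground glows selected by vertex-color green.
# Flares are camera-facing sprites with optional sine flicker; both parts use radial falloffs, camera-relative offsets and depth fade.
# Per-instance data varies flicker phase and color, and brightness compensates for eye adaptation.
# A FadeOut from camera height and time of day dims lights and collapses vertices to cull them.
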Material /Game/Environments/Asteria/Sets/Village/Blueprints/Volumetric/M_Rainforest_FarLights
graph {
 properties {
  bUsedWithInstancedStaticMeshes=true
  bUsedWithNanite=true
  blendMode=BLEND_Additive
  pixelDepthOffsetMode=PDOM_Legacy
  shadingModel=MSM_Unlit
  usedShadingModels=Unlit
 }
 nodes [
  n1 = Constant()
  n2 = MaterialFunctionCall() {materialFunction=&ComponentPivotLocation}
  n3 = Multiply(A=Add(A=Time(), B=PerInstanceCustomData!(DefaultValue)), B=ScalarParameter() {defaultValue=1, parameterName=FlickerSpeed})
  n4 = Max(A=Sine(None=@n3) {period=0.8}, B=Sine(None=@n3))
  n5 = ScalarParameter() {defaultValue=5, parameterName=FlickerMotion_FadeOut}
  n6 = ScalarParameter() {defaultValue=20, parameterName=FlickerMotion}
  n7 = NamedRerouteUsage()
  n8 = Constant()
  n9 = MaterialFunctionCall(X=Multiply(A=@n4, B=LinearInterpolate(A=@n5, B=@n6, Alpha=@n7)), Y=@n8, Z=@n8) {materialFunction=&MakeFloat3}
  n10 = StaticSwitchParameter(True=Add(A=@n2.Result, B=Transform(None=@n9.Result) {transformSourceType=TRANSFORMSOURCE_Camera}), False=@n2.Result) {parameterName=UseFlicker}
  Add(
   A=MaterialFunctionCall(X=@n1, Y=@n1, Z=ScalarParameter() {parameterName=FlarePivotZOffset}) {materialFunction=&MakeFloat3}.Result
   B=@n10
  )
  n11 = NamedRerouteUsage()
  n12 = VertexInterpolator(VS=PerInstanceCustomData())
  n13 = LinearInterpolate(
   A=VectorParameter() {parameterName=ColorCenter, defaultValue=<rgba 3 3 3 1>}.RGB
   B=VectorParameter() {parameterName=ColorCenter2, defaultValue=<rgba 0.8656 1.2635 3 1>}.RGB
   Alpha=@n12.PS
  )
  n14 = ScalarParameter() {defaultValue=0.85, parameterName=BrightnessFlare}
  n15 = Distance(A=TextureCoordinate(), B=Constant2Vector() {g=0.5, r=0.5})
  n16 = Saturate(None=OneMinus(None=Divide(A=@n15) {constB=0.5}))
  n17 = Power(
   Base=@n16
   Exp=LinearInterpolate(
    A=ScalarParameter() {defaultValue=7, parameterName=FalloffCenter_FadeOut}
    B=ScalarParameter() {defaultValue=32, parameterName=FalloffCenter}
    Alpha=NamedRerouteUsage()
   )
  )
  n18 = Multiply(
   A=Power(Base=@n16, Exp=ScalarParameter() {defaultValue=2.5273, parameterName=FalloffMedium})
   B=ScalarParameter() {defaultValue=0.0746, parameterName=BirghtnessMedium}
  )
  n19 = LinearInterpolate(
   A=VectorParameter() {parameterName=Color, defaultValue=<rgba 0.18 0.1519 0.1768 1>}.RGB
   B=VectorParameter() {parameterName=Color2, defaultValue=<rgba 0.0582 0.0822 0.2083 1>}.RGB
   Alpha=@n12.PS
  )
  n20 = ScalarParameter() {defaultValue=0.5, parameterName=BrightnessGround}
  n21 = LinearInterpolate(
   A=ScalarParameter() {defaultValue=2, parameterName=Falloff_FadeOut}
   B=ScalarParameter() {defaultValue=2, parameterName=Falloff}
   Alpha=NamedRerouteUsage()
  )
  n22 = VertexColor()
  n23 = LinearInterpolate(
   A=Multiply(A=Multiply(A=@n13, B=@n14), B=Add(A=@n17, B=@n18))
   B=Multiply(A=Multiply(A=@n19, B=@n20), B=Power(Base=@n16, Exp=@n21))
   Alpha=@n22.G
  )
  Divide(A=Multiply(A=Multiply(A=@n23, B=@n11), B=ScalarParameter() {defaultValue=1.2, parameterName=Exposure}), B=EyeAdaptation().EyeAdaptation)
  n24 = ScalarParameter() {defaultValue=2000, parameterName=Size}
  n25 = StaticSwitchParameter(True=LinearInterpolate(A=ScalarParameter() {defaultValue=200, parameterName=Size_FadeOut}, B=@n24, Alpha=NamedRerouteUsage()), False=@n24) {parameterName=UseCollapse}
  n26 = Multiply(
   A=@n4
   B=LinearInterpolate(
    A=ScalarParameter() {defaultValue=50, parameterName=FlickerSize_FadeOut}
    B=ScalarParameter() {defaultValue=200, parameterName=FlickerSize}
    Alpha=@n7
   )
  )
  n27 = MaterialFunctionCall!(
   Center Location=@n10
   0-1 UVs
   XY scale=StaticSwitchParameter(True=Add(A=@n25, B=@n26), False=@n25) {parameterName=UseFlicker}
   OverridePivot
   "Normals (Optional)"
  ) {materialFunction=&Sprite}
  n28 = ScalarParameter() {defaultValue=-100, parameterName=OffsetFlare}
  n29 = Subtract(A=CameraPositionWS(), B=WorldPosition().XYZ)
  n30 = Constant3Vector() {constant=<rgba 1 1 0 0>}
  n31 = Normalize(VectorInput=@n29)
  n32 = Constant3Vector() {constant=<rgba 0 0 1 0>}
  n33 = Multiply(A=Normalize(VectorInput=Multiply(A=@n29, B=@n30)), B=OneMinus(None=DotProduct(A=@n31, B=@n32)))
  n34 = StaticSwitchParameter(True=Add(A=@n27.World position offset, B=Multiply(A=@n28, B=@n33)), False=@n27.World position offset) {parameterName=UseOffset}
  n35 = StaticSwitchParameter(True=Multiply(A=ScalarParameter() {defaultValue=3000, parameterName=Offset}, B=@n33), False=Constant()) {parameterName=UseOffset}
  n36 = VertexColor()
  n37 = Multiply(A=@n31, B=ScalarParameter() {defaultValue=10, parameterName=CameraDirectionOffset})
  n38 = ObjectPositionWS()
  n39 = WorldPosition()
  n40 = Add(A=@n37, B=LinearInterpolate(A=Subtract(A=@n38, B=@n39.XYZ), Alpha=NamedRerouteUsage()) {constB=0})
  n41 = NamedRerouteUsage()
  n42 = ScalarParameter() {parameterName=CullBias}
  LinearInterpolate(
   A=Subtract(A=ObjectPositionWS(), B=WorldPosition().XYZ)
   B=Add(A=LinearInterpolate(A=@n34, B=@n35, Alpha=@n36.G), B=StaticSwitchParameter(True=@n40, False=@n37) {parameterName=UseCollapse})
   Alpha=Ceil(None=Subtract(A=@n41, B=@n42))
  )
  n43 = Constant() {r=1}
  n44 = DepthFade(Opacity=@n43, FadeDistance=ScalarParameter() {defaultValue=453.4101, parameterName=DepthFade})
  Multiply(A=StaticSwitchParameter(True=@n44, False=LinearInterpolate(A=@n43, B=@n44, Alpha=@n22.G)) {parameterName=UseDepthFade For both}, B=@n11)
  n45 = CameraPositionWS()
  n46 = ScalarParameter() {defaultValue=-14428.2754, parameterName=CameraHeightOffset_Flare}
  n47 = ScalarParameter() {defaultValue=-9632.6338, parameterName=CameraHeightOffset}
  n48 = VertexColor()
  n49 = Divide(
   A=Add(A=ComponentMask(None=@n45) {b=true}, B=LinearInterpolate(A=@n46, B=@n47, Alpha=@n48.G))
   B=LinearInterpolate(
    A=ScalarParameter() {defaultValue=49642.7188, parameterName=CameraHeightFade_Flare}
    B=ScalarParameter() {defaultValue=10000, parameterName=CameraHeightFade}
    Alpha=@n48.G
   )
  )
  n50 = ScalarParameter() {parameterName=Morning_Evening}
  n51 = ScalarParameter() {parameterName=Day}
  n52 = ScalarParameter() {defaultValue=1, parameterName=Night}
  NamedRerouteDeclaration(None=Multiply(
    A=Saturate(None=@n49)
    B=MaterialFunctionCall(Value_Morning=@n50, Value_Day=@n51, Value_Evening=@n50, Value_Night=@n52) {materialFunction=&Scalar_Time_of_Day}.Result
   )) {name=FadeOut, nodeColor=<rgba 1 0.3984 0 1>}
 ]
}
